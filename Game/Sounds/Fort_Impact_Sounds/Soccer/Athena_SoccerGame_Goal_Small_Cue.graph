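# Sound cue for a small goal score in the Athena soccer game, spatialized with a soccer score attenuation.
# A quality-level node plays a layered mix for its first two slots and a single optimized goal wave for the other two.
# The mix blends the main goal wave with a quieter soccer ball toy pop, which a modulator pitches up by a fixed amount.
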SoundCue /Game/Sounds/Fort_Impact_Sounds/Soccer/Athena_SoccerGame_Goal_Small_Cue
root=QualityLevel_1
nodes [
 {id=Mixer_0, children [WavePlayer_1,Modulator_0], properties {inputVolume [1,0.75]}}
 {id=Modulator_0, children [WavePlayer_4], properties {pitchMax=1.2, pitchMin=1.2, volumeMax=1, volumeMin=1}}
 {id=QualityLevel_1, children [Mixer_0,Mixer_0,WavePlayer_0,WavePlayer_0]}
 {id=WavePlayer_0, properties {soundWaveAssetPtr=&/Soccer_Goal_01_OPT_02}}
 {id=WavePlayer_1, properties {soundWaveAssetPtr=&/Soccer_Goal_01}}
 {id=WavePlayer_4, properties {soundWaveAssetPtr=/Game/Athena/Sounds/Toys/SoccerBall/SoccerBallToy_Goal_Pop_01}}
]
properties {
 attenuationSettings=/Game/Sounds/Attenuation/Fort_Soccer_Score_Athena_Att
 maxDistance=7000
 soundClassObject=&Fort_Ambient_Indoors_Athena
 volumeMultiplier=0.501
}
