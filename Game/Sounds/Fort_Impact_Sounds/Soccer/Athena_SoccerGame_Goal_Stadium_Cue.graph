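# Sound cue for a soccer goal in an Athena stadium, using the indoor ambient sound class.
# A quality-level switch plays the stadium goal sound mixed with a quieter toy-ball goal pop on the first two levels, and a single optimized stadium wave on the other two.
# That shared branch feeds both a non-spatial and a stadium attenuation node, which the root mixer layers at equal volume.
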SoundCue /Game/Sounds/Fort_Impact_Sounds/Soccer/Athena_SoccerGame_Goal_Stadium_Cue
root=Mixer_1
nodes [
 {id=Attenuation_0, children [QualityLevel_0], properties {attenuationSettings=/Game/Sounds/Attenuation/Fort_Soccer_NoSpatial_Athena_Att}}
 {id=Attenuation_1, children [QualityLevel_0], properties {attenuationSettings=/Game/Sounds/Attenuation/Fort_Soccer_Stadium_Athena_Att}}
 {id=Mixer_0, children [WavePlayer_1,WavePlayer_2], properties {inputVolume [1,0.2]}}
 {id=Mixer_1, children [Attenuation_0, Attenuation_1], properties {inputVolume [1,1]}}
 {id=QualityLevel_0, children [Mixer_0,Mixer_0,WavePlayer_0,WavePlayer_0]}
 {id=WavePlayer_0, properties {soundWaveAssetPtr=&/Soccer_Ball_Goal_Stadium_02_opt}}
 {id=WavePlayer_1, properties {soundWaveAssetPtr=&/Soccer_Ball_Goal_Stadium_02}}
 {id=WavePlayer_2, properties {soundWaveAssetPtr=/Game/Athena/Sounds/Toys/SoccerBall/SoccerBallToy_Goal_Pop_01}}
]
properties {maxDistance=15383.4258, soundClassObject=&Fort_Ambient_Indoors_Athena, volumeMultiplier=0.702}
