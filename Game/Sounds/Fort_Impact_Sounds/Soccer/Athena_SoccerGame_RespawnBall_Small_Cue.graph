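# Sound cue for respawning a small soccer ball in the Athena soccer game, layering a spawn sound with a delayed firework pop.
# A mixer combines a quality-level switch, which picks the standard or optimized spawn wave, with a boosted Roman candle explosion that starts after a fixed one-second delay.
# The cue overrides attenuation for spatial falloff and routes to an indoor ambient sound class.
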SoundCue /Game/Sounds/Fort_Impact_Sounds/Soccer/Athena_SoccerGame_RespawnBall_Small_Cue
root=Mixer_1
nodes [
 {id=Delay_0, children [WavePlayer_3], properties {delayMax=1, delayMin=1}}
 {id=Mixer_1, children [QualityLevel_0, Delay_0], properties {inputVolume [1,4]}}
 {id=QualityLevel_0, children [WavePlayer_0,WavePlayer_0,WavePlayer_1,WavePlayer_1]}
 {id=WavePlayer_0, properties {soundWaveAssetPtr=&/Soccer_Ball_Spawn_01}}
 {id=WavePlayer_1, properties {soundWaveAssetPtr=&/Soccer_Ball_Spawn_01_opt}}
 {id=WavePlayer_3, properties {soundWaveAssetPtr=/Game/Sounds/Fort_GamePlay_Sounds/Explosives/roman_candle_explo_mono_03}}
]
properties {
 attenuationSettings=&Fort_Weapon_Projectile_Close_Athena_Att
 bOverrideAttenuation=true
 maxDistance=2012
 soundClassObject=&Fort_Ambient_Indoors_Athena
 volumeMultiplier=0.501
 attenuationOverrides {falloffDistance=1500, stereoSpread=2000, attenuationShapeExtents {x=512}}
}
