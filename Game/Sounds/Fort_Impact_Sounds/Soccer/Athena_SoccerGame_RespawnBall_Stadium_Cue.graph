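# Sound cue for the ball respawn sound in the Athena soccer game's stadium.
# A quality-level node picks either the standard stadium spawn wave or its "opt" variant.
# That one node feeds two parallel attenuation branches, one using a non-spatial soccer preset and the other a stadium preset.
# A mixer blends both branches equally, layering the same wave under two distance behaviors.
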SoundCue /Game/Sounds/Fort_Impact_Sounds/Soccer/Athena_SoccerGame_RespawnBall_Stadium_Cue
root=Mixer_1
nodes [
 {id=Attenuation_0, children [QualityLevel_0], properties {attenuationSettings=/Game/Sounds/Attenuation/Fort_Soccer_NoSpatial_Athena_Att}}
 {id=Attenuation_1, children [QualityLevel_0], properties {attenuationSettings=/Game/Sounds/Attenuation/Fort_Soccer_Stadium_Athena_Att}}
 {id=Mixer_1, children [Attenuation_0, Attenuation_1], properties {inputVolume [1,1]}}
 {id=QualityLevel_0, children [WavePlayer_2,WavePlayer_2,WavePlayer_1,WavePlayer_1]}
 {id=WavePlayer_1, properties {soundWaveAssetPtr=&/Soccer_Ball_Spawn_Stadium_01_opt}}
 {id=WavePlayer_2, properties {soundWaveAssetPtr=&/Soccer_Ball_Spawn_Stadium_01}}
]
properties {
 maxDistance=15383.4258
 soundClassObject=&Fort_Ambient_Indoors_Athena
 volumeMultiplier=0.501
 attenuationOverrides {falloffDistance=1500, stereoSpread=2000, attenuationShapeExtents {x=4000}}
}
