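# Sound cue named for the score-up sound in the Athena soccer game, but it has no nodes and a null root, so it holds no audio.
# It carries only playback settings: a soccer attenuation asset named as non-spatial, the indoor ambient Athena sound class, a raised volume multiplier and a max distance.
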SoundCue /Game/Sounds/Fort_Impact_Sounds/Soccer/Athena_SoccerGame_ScoreUp_Cue
root=null
nodes []
properties {
 attenuationSettings=/Game/Sounds/Attenuation/Fort_Soccer_NoSpatial_Athena_Att
 maxDistance=6383.4253
 soundClassObject=&Fort_Ambient_Indoors_Athena
 volumeMultiplier=1.504
}
