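# Sound cue that plays the soccer game-won sound in Athena, as the small variant.
# A quality-level node at the root picks the standard wave for the first two levels and an OPT wave for the last two.
# Playback is spatialized by the soccer score attenuation asset, routed to an indoor ambient sound class, at reduced volume.
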SoundCue /Game/Sounds/Fort_Impact_Sounds/Soccer/Athena_SoccerGame_WonGame_Small_Cue
root=QualityLevel_0
nodes [
 {id=QualityLevel_0, children [WavePlayer_0,WavePlayer_0,WavePlayer_2,WavePlayer_2]}
 {id=WavePlayer_0, properties {soundWaveAssetPtr=&/Soccer_Won_Game}}
 {id=WavePlayer_2, properties {soundWaveAssetPtr=&/Soccer_Won_Game_OPT}}
]
properties {
 attenuationSettings=/Game/Sounds/Attenuation/Fort_Soccer_Score_Athena_Att
 maxDistance=7000
 soundClassObject=&Fort_Ambient_Indoors_Athena
 volumeMultiplier=0.501
}
