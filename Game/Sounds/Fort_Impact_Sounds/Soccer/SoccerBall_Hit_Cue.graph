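# Sound cue for soccer ball hit impacts, classed as an Athena ambient prop sound.
# A Random node picks one of three equally weighted hit waves and feeds a Modulator at the root for pitch and volume variation.
# It uses a soccer ball bounce attenuation asset and overrides concurrency to cap simultaneous instances at three.
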SoundCue /Game/Sounds/Fort_Impact_Sounds/Soccer/SoccerBall_Hit_Cue
root=Modulator_0
nodes [
 {id=Modulator_0, children [Random_0]}
 {id=Random_0, children [WavePlayer_1,WavePlayer_2,WavePlayer_3], properties {weights [1,1,1]}}
 {id=WavePlayer_1, properties {soundWaveAssetPtr=&/SoccerBall_Hit_01}}
 {id=WavePlayer_2, properties {soundWaveAssetPtr=&/SoccerBall_Hit_02}}
 {id=WavePlayer_3, properties {soundWaveAssetPtr=&/SoccerBall_Hit_03}}
]
properties {
 attenuationSettings=/Game/Sounds/Attenuation/Fort_Ball_Soccer_Bounce_Athena_Att
 bOverrideConcurrency=true
 maxDistance=3010
 soundClassObject=&Fort_Ambient_Props_Athena
 volumeMultiplier=1
 concurrencyOverrides {maxCount=3}
}
