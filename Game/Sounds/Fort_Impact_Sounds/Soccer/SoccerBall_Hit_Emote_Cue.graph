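# Sound cue for a soccer ball hit, apparently played during an emote.
# It picks one of three SoccerBall_Hit waves at equal weight, then a modulator slightly varies pitch and volume on each play.
# The cue plays at reduced volume through the emote music sound class with Athena emote attenuation, limiting it to nearby listeners.
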SoundCue /Game/Sounds/Fort_Impact_Sounds/Soccer/SoccerBall_Hit_Emote_Cue
root=Modulator_0
nodes [
 {id=Modulator_0, children [Random_0], properties {pitchMax=1.1, pitchMin=0.93, volumeMax=1, volumeMin=0.85}}
 {id=Random_0, children [WavePlayer_1,WavePlayer_2,WavePlayer_3], properties {weights [1,1,1]}}
 {id=WavePlayer_1, properties {soundWaveAssetPtr=&/SoccerBall_Hit_01}}
 {id=WavePlayer_2, properties {soundWaveAssetPtr=&/SoccerBall_Hit_02}}
 {id=WavePlayer_3, properties {soundWaveAssetPtr=&/SoccerBall_Hit_03}}
]
properties {
 attenuationSettings=&Fort_Emotes_Athena_Att
 maxDistance=1600
 soundClassObject=&Fort_Music_Emote
 volumeMultiplier=0.5
}
