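# Sound cue for a soccer ball hit, routed through the Fort_Impacts sound class.
# A Random node picks one of three hit waves with equal weight, then feeds a Modulator left at default settings for the output.
# It uses the shared Athena ball-bounce attenuation asset for spatial falloff, with a slightly boosted cue volume.
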SoundCue /Game/Sounds/Fort_Impact_Sounds/Soccer/SoccerBall_Hit_STW_Cue
root=Modulator_0
nodes [
 {id=Modulator_0, children [Random_0]}
 {id=Random_0, children [WavePlayer_1,WavePlayer_2,WavePlayer_3], properties {weights [1,1,1]}}
 {id=WavePlayer_1, properties {soundWaveAssetPtr=&/SoccerBall_Hit_01}}
 {id=WavePlayer_2, properties {soundWaveAssetPtr=&/SoccerBall_Hit_02}}
 {id=WavePlayer_3, properties {soundWaveAssetPtr=&/SoccerBall_Hit_03}}
]
properties {
 attenuationSettings=/Game/Sounds/Attenuation/Fort_Ball_Bounce_Athena_Att
 maxDistance=7000
 soundClassObject=&Fort_Impacts
 volumeMultiplier=1.2096
}
